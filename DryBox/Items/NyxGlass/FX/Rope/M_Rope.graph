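# Rope material for the NyxGlass gadget, drawing a textured, glowing rope between start and end points.
# World position offset layers helix swing, Perlin noise and arrival shake, shaped by curve-atlas rows along length and activation/arrival timelines.
# Emissive combines base texture, traveling shine and faked GGX highlights.
# Static switches select a Niagara ribbon version, using dynamic parameters with optional noise-only offset, or a SplineThicken glow shell with edge-masked opacity.
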Material /DryBox/Items/NyxGlass/FX/Rope/M_Rope
graph {
 properties {
  bEnableResponsiveAA=true
  bUsedWithNiagaraRibbons=true
  pixelDepthOffsetMode=PDOM_Legacy
  translucencyPass=MTP_AfterMotionBlur
  usedShadingModels=Default Lit
 }
 nodes [
  Constant3Vector() {constant=<rgba 0 0 1 0>}
  Constant()
  Constant() {r=0.4}
  Constant() {r=0.05}
  n1 = Constant() {r=1}
  n2 = NamedRerouteUsage()
  n3 = Multiply(
   A=ScalarParameter() {defaultValue=2, parameterName=RopeTexture_ScaleV}
   B=StaticSwitchParameter(
    True=Divide(A=@n2) {constB=100}
    False=MaterialFunctionCall() {materialFunction=/Engine/Functions/Engine_MaterialFunctions02/WorldPositionOffset/ObjectScale}.Scale Z
   ) {parameterName=NiagaraVersion}
  )
  n4 = MaterialFunctionCall(XY=Multiply(A=NamedRerouteUsage(), B=AppendVector(A=@n1, B=@n3))) {materialFunction=&Swizzle}
  n5 = TextureSample(UVs=@n4.YX) {texture=/DryBox/Items/NyxGlass/Textures/T_Gadget_NyxGlass_D_FX}
  n6 = Multiply(A=@n5.RGB, B=ScalarParameter() {parameterName=BaseColorScaleForEmi})
  n7 = ScalarParameter() {parameterName=Shine_TotalE}
  n8 = Constant()
  n9 = OneMinus(None=MaterialFunctionCall(Float2=NamedRerouteUsage()) {materialFunction=&BreakOutFloat2Components}.G)
  n10 = NamedRerouteUsage()
  n11 = NamedRerouteUsage()
  n12 = ScalarParameter() {defaultValue=-0.1, parameterName=Shine_StartMoveTimeOffset}
  n13 = Saturate(None=Multiply(A=Add(A=@n11, B=@n12), B=ScalarParameter() {defaultValue=1.5, parameterName=Shine_MoveSpeed}) {constB=5})
  n14 = Subtract(A=Multiply(A=@n9, B=@n10), B=LinearInterpolate(B=@n10, Alpha=OneMinus(None=@n13))) {constB=300}
  n15 = ScalarParameter() {defaultValue=100, parameterName=Shine_SoftWidth}
  n16 = OneMinus(None=Saturate(None=Divide(A=@n14, B=@n15)))
  n17 = ScalarParameter() {defaultValue=-3000, parameterName=Shine_Width}
  n18 = Saturate(None=Divide(A=Subtract(A=@n14, B=@n17), B=@n15))
  n19 = Multiply(A=@n9, B=ScalarParameter() {defaultValue=1.25, parameterName=ShineVFadeM})
  n20 = Multiply(
   A=Power(Base=Multiply(A=@n16, B=@n18), Exp=ScalarParameter() {defaultValue=1.5, parameterName=Shine_SoftPow})
   B=Power(Base=Saturate(None=@n19), Exp=ScalarParameter() {defaultValue=8, parameterName=Shine_VFadePow})
  )
  n21 = VectorParameter() {parameterName=ShineColor, defaultValue=<rgba 6.5 22.0833 100 0>}
  n22 = StaticSwitchParameter(True=@n8, False=Multiply(A=Saturate(None=@n20), B=@n21.RGB) {constB=0}) {parameterName=NiagaraVersion}
  n23 = NamedRerouteUsage()
  n24 = ScalarParameter() {defaultValue=-0.04, parameterName=GlowSpline_FadeOffset}
  n25 = Saturate(None=Multiply(A=Add(A=@n23, B=@n24), B=ScalarParameter() {defaultValue=3, parameterName=GlowSpline_FadeSpeed}) {constB=5})
  n26 = Multiply(
   A=Add(A=@n22, B=VectorParameter() {parameterName=GlowSpline_EmissiveColor, defaultValue=<rgba 0.075 0.1042 0.25 1>}.RGB)
   B=StaticSwitchParameter(True=Constant() {r=1}, False=OneMinus(None=@n25)) {parameterName=NiagaraVersion}
  )
  n27 = Constant3Vector() {constant=<rgba 0 0 1 0>}
  n28 = MaterialFunctionCall!(
   Plane Normal=@n27
   Vector=MaterialFunctionCall() {materialFunction=&CameraDirectionVector}.Result
  ) {materialFunction=/Engine/Functions/Engine_MaterialFunctions02/Math/ProjectVectorOntoPlane}
  n29 = MaterialFunctionCall(Plane Normal=@n27, Vector=SkyAtmosphereLightDirection()) {materialFunction=&ProjectVectorOntoPlane}
  n30 = Abs(None=DotProduct(A=Normalize(VectorInput=@n28.Result), B=Normalize(VectorInput=@n29.Result)))
  n31 = Multiply(A=NamedRerouteUsage(), B=ScalarParameter() {defaultValue=0.5, parameterName=FakeGGX_FadeInSpeed}) {constB=5}
  n32 = Transform(None=Constant3Vector() {constant=<rgba 0 0 1 0>}) {transformSourceType=TRANSFORMSOURCE_Local}
  n33 = DotProduct(
   A=MaterialFunctionCall() {materialFunction=&CameraDirectionVector}.Result
   B=StaticSwitchParameter(True=NamedRerouteUsage(), False=Normalize(VectorInput=@n32)) {parameterName=NiagaraVersion}
  )
  n34 = ScalarParameter() {defaultValue=0.5, parameterName=FakeGGX_AngleMoveSpeed}
  n35 = Add(
   A=MaterialFunctionCall(Float2=NamedRerouteUsage()) {materialFunction=&BreakOutFloat2Components}.G
   B=Add(A=Multiply(A=@n33, B=@n34), B=ScalarParameter() {defaultValue=-0.2, parameterName=FakeGGX_AngleMoveOffset})
  )
  n36 = VectorParameter() {group=FakeGGX, parameterName=FakeGGX_EdgeMask_Control, channelNames {a="", b=Power, g=Multiply, r=Offset}, defaultValue=<rgba -0.2 3 6 1>}
  n37 = StaticSwitchParameter(
   True=CrossProduct(A=VertexNormalWS(), B=VertexTangentWS())
   False=Transform(None=Constant3Vector() {constant=<rgba 0 0 1 1>}) {transformSourceType=TRANSFORMSOURCE_Local}
  ) {parameterName=NiagaraVersion}
  n38 = Normalize(VectorInput=MaterialFunctionCall(Plane Normal=Normalize(VectorInput=@n37), Vector=CameraVectorWS()) {materialFunction=&ProjectVectorOntoPlane}.Result)
  n39 = PixelNormalWS()
  n40 = Saturate(None=Abs(None=DotProduct(A=@n38, B=@n39)))
  n41 = ScalarParameter() {defaultValue=15, parameterName=FakeGGX_FresnelPow}
  n42 = Multiply(
   A=StaticSwitchParameter(True=Constant() {r=1}, False=Saturate(None=@n31)) {parameterName=NiagaraVersion}
   B=Multiply(
    A=MaterialFunctionCall!(UorV=@n35, Invert_EdgeControl=@n36, NoInvert_EdgeControl=@n36) {materialFunction=/Game/Effects/Niagara/MasterMaterials/Functions/Common/FXCommon_EdgeMask_Ex}.Mask
    B=Power(Base=@n40, Exp=@n41) {constExponent=50}
   )
  )
  n43 = ScalarParameter() {defaultValue=20, parameterName=FakeGGX_ScaleEmissive}
  n44 = Add(
   A=Multiply(A=Power(Base=@n30) {constExponent=0.75}, B=Multiply(A=@n42, B=@n43))
   B=MaterialFunctionCall!(LightDirctionIn, CameraVectorIn, Roughness=ScalarParameter() {defaultValue=0.2, parameterName=GGX_Roughness}, NormalIn) {materialFunction=/Game/Effects/Niagara/MasterMaterials/Functions/Common/FXCommon_GGX}.Result
  )
  n45 = Power(Base=@n40, Exp=ScalarParameter() {defaultValue=7, parameterName=Shine_FresnelPower}) {constExponent=50}
  n46 = Add(
   A=Multiply(A=VectorParameter() {parameterName=GGX_Color, defaultValue=<rgba 2.5 2.9167 5 0>}.RGB, B=Saturate(None=@n44))
   B=LinearInterpolate(A=@n22, B=Multiply(A=@n22, B=@n45))
  )
  n47 = VectorParameter() {parameterName=Emissive_RadiusAndRange, defaultValue=<rgba 0 1 1000 2000>}
  n48 = Subtract(A=NamedRerouteUsage(), B=@n47.B) {constB=1000}
  n49 = LinearInterpolate(A=@n47.R, B=@n47.R, Alpha=Saturate(None=Divide(A=@n48, B=@n47.A) {constB=3000}))
  n50 = StaticSwitchParameter(True=@n49, False=Constant() {r=1}) {parameterName=NiagaraDistanceEmissive}
  n51 = Add(
   A=Multiply(A=VectorParameter() {parameterName=ShineColor_HighLight, defaultValue=<rgba 2.5 18.3333 50 0>}.RGB, B=Multiply(B=@n7) {constA=0.3})
   B=Multiply(
    A=StaticSwitchParameter(True=@n26, False=@n46) {parameterName=SplineThicken?}
    B=StaticSwitchParameter(True=@n50, False=@n49) {parameterName=NiagaraVersion}
   )
  )
  MaterialFunctionCall(Emissive=Divide(A=Add(A=@n6, B=@n51)) {constB=1.9}) {materialFunction=&MF_FX_EmissiveAdjustments}
  n52 = Multiply(A=Subtract(A=Time(), B=ScalarParameter() {defaultValue=-100, parameterName=Shine_AttackTime})) {constB=3}
  Multiply(A=Power(Base=Saturate(None=@n52)) {constExponent=15}, B=@n7)
  Multiply(A=@n5.RGB, B=ScalarParameter() {defaultValue=0.3, parameterName=BaseColorScale})
  n53 = DynamicParameter() {parameterIndex=3, paramNames [EndX,EndY,EndZ,WPOStr]}
  MaterialExpressionAppend3Vector_0 = MaterialXAppend3Vector!(A=@n53.EndX, B=@n53.EndX, C=@n53.EndX)
  n54 = DynamicParameter() {parameterIndex=2, paramNames [StartX,StartY,StartZ,Param4]}
  MaterialExpressionAppend3Vector_1 = MaterialXAppend3Vector(A=@n54.StartX, B=@n54.StartX, C=@n54.StartX)
  n55 = Subtract(
   A=StaticSwitchParameter(True=@MaterialExpressionAppend3Vector_0, False=VectorParameter() {parameterName=RopeEndPos, defaultValue=<rgba -2735 606 669 1>}.RGB) {parameterName=NiagaraVersion}
   B=StaticSwitchParameter(True=@MaterialExpressionAppend3Vector_1, False=VectorParameter() {parameterName=RopeStartPos, defaultValue=<rgba -2735 -1000 669 1>}.RGB) {parameterName=NiagaraVersion}
  )
  NamedRerouteDeclaration(None=Length(None=@n55) {is=class MaterialExpressionLength}) {name=Rope_Length, nodeColor=<rgba 0.9688 1 0 1>}
  NamedRerouteDeclaration(None=Normalize(VectorInput=@n55)) {name=Rope_Direction, nodeColor=<rgba 0.7109 0 1 1>}
  n56 = TextureCoordinate()
  NamedRerouteDeclaration(None=StaticSwitchParameter(True=MaterialFunctionCall(XY=@n56) {materialFunction=&Swizzle}.YX, False=@n56) {parameterName=NiagaraVersion}) {name=UV, nodeColor=<rgba 1 0.5625 0 1>}
  NamedRerouteDeclaration(None=ScalarParameter() {parameterName=IsRight}) {name=Rope_IsRight, nodeColor=<rgba 0.8281 1 0 1>}
  NamedRerouteDeclaration(None=Subtract(A=Time() {period=6000}, B=ScalarParameter() {defaultValue=-100, parameterName=ArriveTime})) {name=Rope_ArriveTimeline, nodeColor=<rgba 0 0.2266 1 1>}
  n57 = MaterialFunctionCall(Float2=NamedRerouteUsage()) {materialFunction=&BreakOutFloat2Components}
  n58 = CurveAtlasRowParameter(CurveTime=@n57.G) {
   atlas=&/CLA_NyxGlass_Rope
   curve=&/CL_NyxGlass_SampleByV
   defaultValue=2
   parameterName=RopeCurve_SampleByV
  }
  n59 = NamedRerouteUsage()
  n60 = ScalarParameter() {defaultValue=1000, parameterName=RopeCurve_RealDistScale}
  n61 = CurveAtlasRowParameter(CurveTime=Multiply(A=@n57.G, B=Divide(A=@n59, B=@n60) {constB=1000})) {
   atlas=&CLA_NyxGlass_Rope
   curve=&/CL_NyxGlass_SampleByRealDist
   parameterName=RopeCurve_SampleByRealDist
  }
  n62 = CurveAtlasRowParameter(CurveTime=NamedRerouteUsage()) {
   atlas=&CLA_NyxGlass_Rope
   curve=&/CL_NyxGlass_SampleByTimeline
   defaultValue=1
   parameterName=RopeCurve_SampleByTimeline
  }
  NamedRerouteDeclaration(None=Multiply(A=Multiply(A=@n58.B, B=@n61.B), B=@n62.B)) {name=ArriveShake_CurveValue, nodeColor=<rgba 0 1 0.4375 1>}
  n63 = DynamicParameter() {parameterIndex=1, paramNames [YAxisX,YAxisY,YAxisZ,Param4]}
  MaterialExpressionAppend3Vector_2 = MaterialXAppend3Vector(A=@n63.YAxisX, B=@n63.YAxisX, C=@n63.YAxisX)
  n64 = Transform(None=Constant3Vector() {constant=<rgba 0 1 0 0>}) {transformSourceType=TRANSFORMSOURCE_Local}
  NamedRerouteDeclaration(None=StaticSwitchParameter(True=@MaterialExpressionAppend3Vector_2, False=Normalize(VectorInput=@n64)) {parameterName=NiagaraVersion}) {name=Rope_YAxis, nodeColor=<rgba 0 0.4141 1 1>}
  n65 = DynamicParameter() {paramNames [XAxisX, XAxisY, XAxisZ, ""]}
  MaterialExpressionAppend3Vector_3 = MaterialXAppend3Vector(A=@n65.XAxisX, B=@n65.XAxisX, C=@n65.XAxisX)
  n66 = Transform(None=Constant3Vector() {constant=<rgba 1 0 0 0>}) {transformSourceType=TRANSFORMSOURCE_Local}
  NamedRerouteDeclaration(None=StaticSwitchParameter(True=@MaterialExpressionAppend3Vector_3, False=Normalize(VectorInput=@n66)) {parameterName=NiagaraVersion}) {name=Rope_XAxis, nodeColor=<rgba 1 0.3516 0 1>}
  NamedRerouteDeclaration(None=Subtract(A=Time(), B=ScalarParameter() {defaultValue=-100, parameterName=ActiveTime})) {name=Rope_ActiveTimeline, nodeColor=<rgba 0.8447 0 1 1>}
  n67 = CurveAtlasRowParameter(CurveTime=NamedRerouteUsage()) {
   atlas=&CLA_NyxGlass_Rope
   curve=&CL_NyxGlass_SampleByTimeline
   defaultValue=1
   parameterName=RopeCurve_SampleByTimeline
  }
  n68 = ScalarParameter() {parameterName=ForCG_Helix}
  NamedRerouteDeclaration(None=Multiply(A=Multiply(A=@n58.R, B=@n61.R), B=LinearInterpolate(A=@n67.R, Alpha=@n68))) {name=Helix_CurveValue, nodeColor=<rgba 0 1 0.6016 1>}
  n69 = ScalarParameter() {parameterName=ForCG_Random}
  NamedRerouteDeclaration(None=Multiply(A=Multiply(A=@n58.G, B=@n61.G), B=LinearInterpolate(A=@n67.G, Alpha=@n69))) {name=Random_CurveValue, nodeColor=<rgba 1 0 0.2344 1>}
  Normalize()
  n70 = Multiply(A=NamedRerouteUsage(), B=ScalarParameter() {defaultValue=1, parameterName=ForCG_StrScale})
  n71 = TextureObjectParameter() {
   group=RandomX
   parameterName=RandomX_Texture
   samplerType=SAMPLERTYPE_LinearGrayscale
   sortPriority=0
   texture=&T_SimplePerlin_Greyscale
  }
  n72 = VectorParameter() {group=RandomX, parameterName=RandomX_UVControl, channelNames {a=VPan, b=UPan, g=VScale, r=UScale}, defaultValue=<rgba 1 1 0 -0.6>}
  n73 = LinearInterpolate(A=Time() {bOverride_Period=true, period=6000}, B=ScalarParameter() {parameterName=ForCG_Time}, Alpha=ScalarParameter() {parameterName=ForCG})
  n74 = NamedRerouteUsage()
  n75 = ScalarParameter() {defaultValue=4000, parameterName=Random_Freq}
  n76 = NamedRerouteUsage()
  n77 = MaterialFunctionCall(
   X=Constant()
   Y=Multiply(
    A=Divide(A=@n74, B=@n75)
    B=MaterialFunctionCall(Float2=@n76) {materialFunction=&BreakOutFloat2Components}.G
   )
  ) {materialFunction=&MakeFloat2}
  n78 = Multiply(A=Subtract(A=MaterialFunctionCall(Texture=@n71, UVControl_RGB=@n72, UVControl_A=@n72, "[Override]TimeInput"=@n73, "[Override]UVInput"=@n77.Result) {materialFunction=&FXCommon_TextureSample}.R) {constB=0.5}) {constB=2}
  n79 = NamedRerouteUsage()
  n80 = NamedRerouteUsage()
  n81 = TextureObjectParameter() {
   group=RandomY
   parameterName=RandomY_Texture
   samplerType=SAMPLERTYPE_LinearGrayscale
   sortPriority=0
   texture=&T_SimplePerlin_Greyscale
  }
  n82 = VectorParameter() {group=RandomY, parameterName=RandomY_UVControl, channelNames {a=VPan, b=UPan, g=VScale, r=UScale}, defaultValue=<rgba 1 1 0 -1.2>}
  n83 = Subtract(A=MaterialFunctionCall(
    Texture=@n81
    AdditionalUVOffset=ScalarParameter() {defaultValue=1.258, parameterName=Random_Offset}
    UVControl_RGB=@n82
    UVControl_A=@n82
    "[Override]TimeInput"=@n73
    "[Override]UVInput"=@n77.Result
   ) {materialFunction=&FXCommon_TextureSample}.R) {constB=0.5}
  n84 = Add(
   A=Multiply(A=Multiply(A=@n78, B=@n79), B=LinearInterpolate(Alpha=@n80) {constA=1, constB=-1})
   B=Multiply(A=Multiply(A=@n83) {constB=2}, B=NamedRerouteUsage())
  )
  n85 = VectorParameter() {parameterName=Random_RadiusAndRange, defaultValue=<rgba 100 400 0 2000>}
  n86 = Divide(A=Subtract(A=NamedRerouteUsage(), B=@n85.B) {constB=1000}, B=@n85.A) {constB=3000}
  n87 = LinearInterpolate(A=LinearInterpolate(A=@n85.R, B=@n85.R, Alpha=Saturate(None=@n86)), B=@n85.G, Alpha=ScalarParameter() {parameterName=ForCG})
  n88 = Multiply(A=Multiply(A=@n70, B=Multiply(A=@n84, B=@n87)), B=DynamicParameter() {parameterIndex=3, paramNames [EndX,EndY,EndZ,WPOStr]}.WPOStr)
  n89 = Constant()
  n90 = NamedRerouteUsage()
  n91 = ScalarParameter() {defaultValue=50, parameterName=ArriveShake_Frequency}
  n92 = NamedRerouteUsage()
  n93 = ScalarParameter() {defaultValue=-600, parameterName=ArriveShake_Speed}
  n94 = NamedRerouteUsage()
  n95 = Add(
   A=Multiply(
    A=Divide(A=@n90, B=@n91)
    B=MaterialFunctionCall(Float2=@n92) {materialFunction=&BreakOutFloat2Components}.G
   )
   B=Add(A=Multiply(A=@n93, B=@n94), B=ScalarParameter() {parameterName=Random})
  )
  n96 = Divide(A=MaterialFunctionCall() {materialFunction=&Pi}.Result) {constA=180, constB=180}
  n97 = Multiply(A=Sine(None=Multiply(A=@n95, B=@n96) {constB=180}), B=LinearInterpolate(Alpha=NamedRerouteUsage()) {constA=1, constB=-1})
  n98 = VectorParameter() {parameterName=ArriveShake_RadiusAndRangeF, defaultValue=<rgba 0 70 100 3000>}
  n99 = Subtract(A=NamedRerouteUsage(), B=@n98.B) {constB=1000}
  n100 = LinearInterpolate(A=@n98.R, B=@n98.R, Alpha=Saturate(None=Divide(A=@n99, B=@n98.A) {constB=3000}))
  n101 = Cosine(None=Multiply(A=@n95, B=@n96) {constB=180})
  n102 = Add(A=Multiply(A=Multiply(A=@n97, B=@n100), B=NamedRerouteUsage()), B=Multiply(A=Multiply(A=@n101, B=@n100), B=NamedRerouteUsage()))
  n103 = NamedRerouteUsage()
  n104 = NamedRerouteUsage()
  n105 = ScalarParameter() {defaultValue=10, parameterName=Helix_Frequency}
  n106 = NamedRerouteUsage()
  n107 = ScalarParameter() {defaultValue=-200, parameterName=Helix_Speed}
  n108 = LinearInterpolate(A=NamedRerouteUsage(), B=ScalarParameter() {parameterName=ForCG_Time}, Alpha=ScalarParameter() {parameterName=ForCG})
  n109 = Add(
   A=Multiply(
    A=Divide(A=@n104, B=@n105)
    B=MaterialFunctionCall(Float2=@n106) {materialFunction=&BreakOutFloat2Components}.G
   )
   B=Add(A=Multiply(A=@n107, B=@n108), B=ScalarParameter() {parameterName=Random})
  )
  n110 = Divide(A=MaterialFunctionCall() {materialFunction=&Pi}.Result) {constA=180, constB=180}
  n111 = Sine(None=Multiply(A=@n109, B=@n110) {constB=180})
  n112 = LinearInterpolate(Alpha=NamedRerouteUsage()) {constA=1, constB=-1}
  n113 = VectorParameter() {parameterName=Helix_RadiusAndRange, defaultValue=<rgba 50 150 0 2000>}
  n114 = Subtract(A=NamedRerouteUsage(), B=@n113.B) {constB=1000}
  n115 = Power(Base=Saturate(None=Divide(A=@n114, B=@n113.A) {constB=3000}), Exp=ScalarParameter() {defaultValue=1, parameterName=Helix_RadiusGradientPow})
  n116 = Multiply(
   A=LinearInterpolate(A=LinearInterpolate(A=@n113.R, B=@n113.R, Alpha=@n115), B=@n113.G, Alpha=ScalarParameter() {parameterName=ForCG})
   B=ScalarParameter() {defaultValue=1, parameterName=EnableHelixSwing}
  )
  n117 = Multiply(A=Multiply(A=Multiply(A=@n111, B=@n112), B=@n116), B=NamedRerouteUsage())
  n118 = Cosine(None=Multiply(A=@n109, B=@n110) {constB=180})
  n119 = Multiply(A=ScalarParameter() {parameterName=Helix_XAxisScale}, B=Multiply(A=Multiply(A=@n118, B=@n116), B=NamedRerouteUsage()))
  n120 = NamedRerouteUsage()
  n121 = ScalarParameter() {defaultValue=1, parameterName=ForCG_StrScale}
  n122 = NamedRerouteUsage()
  n123 = ScalarParameter() {parameterName=Helix_FadeTimeOffset}
  n124 = Saturate(None=Multiply(A=Add(A=@n122, B=@n123), B=ScalarParameter() {defaultValue=5, parameterName=Helix_FadeSpeed}) {constB=5})
  n125 = Multiply(
   A=Multiply(A=Add(A=@n117, B=@n119), B=Multiply(A=@n120, B=@n121))
   B=LinearInterpolate(A=OneMinus(None=@n124), Alpha=ScalarParameter() {parameterName=ForCG_Helix})
  )
  n126 = ScalarParameter() {defaultValue=1, parameterName=Helix_BendFactor}
  n127 = ScalarParameter() {parameterName=EnableRandomSwing}
  n128 = ScalarParameter() {parameterName=ForCG_Random}
  n129 = Multiply(A=@n84, B=Multiply(A=@n87, B=LinearInterpolate(A=@n127, Alpha=@n128)))
  n130 = Multiply(
   A=Add(A=NamedRerouteUsage(), B=ScalarParameter() {defaultValue=-0.07, parameterName=Random_FadeTimeOffset})
   B=ScalarParameter() {defaultValue=5, parameterName=Random_FadeSpeed}
  ) {constB=5}
  n131 = LinearInterpolate(A=OneMinus(None=Saturate(None=@n130)), Alpha=ScalarParameter() {parameterName=ForCG_Random})
  n132 = Add(A=Multiply(A=@n70, B=Multiply(A=@n129, B=@n131)), B=Add(A=Multiply(A=@n102, B=@n103), B=Multiply(A=@n125, B=@n126)))
  n133 = ScalarParameter() {defaultValue=65, parameterName=GlowSpline_Thicken}
  n134 = Add(A=@n132, B=WorldPosition().XYZ)
  n135 = StaticBool() {value=true}
  n136 = StaticSwitchParameter(
   True=Add(
    A=@n132
    B=MaterialFunctionCall(WidthBase=@n133, WidthTip=@n133, WorldPosition=@n134, Expand U or V UV Channel=@n135) {materialFunction=&SplineThicken}.WorldPositionOffset
   )
   False=@n132
  ) {parameterName=SplineThicken?}
  n137 = Multiply(A=VertexNormalWS(), B=ScalarParameter() {parameterName=wpov})
  Reroute(None=StaticSwitchParameter(True=StaticSwitchParameter(True=@n88, False=@n89) {parameterName=NiagaraWPO}, False=Add(A=@n136, B=@n137)) {parameterName=NiagaraVersion})
  StaticSwitchParameter(True=Constant() {r=1}, False=@n40) {parameterName=NiagaraVersion}
  n138 = MaterialFunctionCall(Float2=NamedRerouteUsage()) {materialFunction=&BreakOutFloat2Components}
  n139 = VectorParameter() {
   group=OpacityTopBottom
   parameterName=OpacityTopBottom_EdgeMask_InvertControl
   channelNames {a="", b=Power, g=Multiply, r=Offset}
   defaultValue=<rgba 0 1.5 2 1>
  }
  n140 = VectorParameter() {
   group=OpacityTopBottom
   parameterName=OpacityTopBottom_EdgeMask_Control
   channelNames {a="", b=Power, g=Multiply, r=Offset}
   defaultValue=<rgba 0 2.5 1 1>
  }
  n141 = VectorParameter() {group=Opacity, parameterName=OpacityEdge_EdgeMask_Control, channelNames {a="", b=Power, g=Multiply, r=Offset}, defaultValue=<rgba 0 2 2 1>}
  n142 = Multiply(
   A=MaterialFunctionCall(UorV=@n138.G, Invert_EdgeControl=@n139, NoInvert_EdgeControl=@n140) {materialFunction=&FXCommon_EdgeMask_Ex}.Mask
   B=MaterialFunctionCall!(UorV=@n138.R, EdgeControl=@n141) {materialFunction=/Game/Effects/Niagara/MasterMaterials/Functions/Common/FXCommon_EdgeMask_Simple}.Mask
  )
  n143 = Constant() {r=1}
  n144 = Constant() {r=1}
  n145 = Multiply(
   A=Multiply(A=StaticSwitchParameter(True=@n142, False=@n143) {parameterName=SplineThicken?}, B=ScalarParameter() {defaultValue=1, parameterName=S})
   B=If(A=PixelDepth().R, B=SceneDepth().R, "A > B"=Constant(), "A == B"=@n144, "A < B"=@n144)
  )
  StaticSwitchParameter(True=Multiply(A=@n145), False=@n145) {parameterName=SplineThicken?}
  n146 = NamedRerouteUsage()
  n147 = OneMinus(None=Saturate(None=Multiply(A=@n146) {constB=3}))
  StaticSwitchParameter(True=@n8, False=Multiply(A=@n21.RGB, B=Power(Base=@n147))) {parameterName=NiagaraVersion}
  Transform(None=MaterialFunctionCall(
    Normal=TextureSample(UVs=@n4.YX) {samplerType=SAMPLERTYPE_Normal, texture=/DryBox/Items/NyxGlass/Textures/T_Gadget_NyxGlass_N_FX}.RGB
    Flatness=ScalarParameter() {defaultValue=1, parameterName=GGX_NormalFalt}
   ) {materialFunction=&FlattenNormal}.Result)
  VectorParameter() {parameterName=RopeTargetPos, defaultValue=<rgba 0 0 0 1>}
  VectorParameter() {
   group=FakeGGX
   parameterName=FakeGGX_EdgeMask_InvertControl
   channelNames {a="", b=Power, g=Multiply, r=Offset}
   defaultValue=<rgba -0.4 3 6 1>
  }
 ]
}
