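# Niagara test rope effect for the NyxGlass item's hook variant, drawn from the emitter origin to User.Target with two ribbon renderers and a sprite renderer.
# A burst of index-ordered particles blends between a straight line and randomized circle/extend offsets from ShrinkPosCalc, driven by system blend values.
# Index-weighted interpolation smooths positions.
# Dynamic material parameters receive axes, start and end positions and an age curve, while randomized alpha scaling adds flicker.
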NiagaraSystem /DryBox/Items/NyxGlass/FX/Rope/NS_ShrinkRopeTest_NyxGlass_HookVer
dependencies {
 usedDataInterfaces [&NiagaraDataInterfaceParticleRead, &NiagaraDataInterfaceCurve]
 usedDynamicInputs [&Divide_Float]
 usedModules [
  &SystemState
  &self:SystemSpawnScript.SystemScriptSource.SystemScriptGraph.NiagaraNodeAssignment_0.SetVariables_#1
 ]
}
emitters {Empty {
  emitterUpdateScript [
   &EmitterState(Inactive Response=::NewEnumerator1, Life Cycle Mode=::NewEnumerator1, Loop Behavior=::NewEnumerator1, Loop Duration=2, UseLoopDelay=false)
   &SpawnBurst_Instantaneous(Spawn Count=20, Use Spawn Probability=false, UseLoopCountLimit=false)
   /SetVariables(
    Emitter.AngleRandom:NiagaraFloat=&Waveform("Waveform [1]"=::NewEnumerator12, "[1] Frequency"=5, "[1] Amplitude Min/Max"=<-1 1>, "[1] Scale"=0.1)
    Emitter.ExtendRandom:NiagaraFloat=&RandomRangeFloat(Minimum=-50, Maximum=50)
    Emitter.RadiusRandom:NiagaraFloat=&RandomRangeFloat(Minimum=-10, Maximum=10)
   ) {isSetParameters=true}
   /SetVariables(Emitter.ColorShaking:NiagaraFloat=&RandomRangeFloat(Minimum=0.1, Maximum=0.2, Evaluation Type=::NewEnumerator1)) {isSetParameters=true}
  ]
  particleSpawnScript [
   &InitializeParticle(
    Ribbon Width=2
    Ribbon Width Mode=::NewEnumerator1
    Sprite Size Mode=::NewEnumerator3
    Uniform Sprite Size=500
    UsePositionOffset=false
    Write Material Random=true
   )
   /SetVariables(Particles.RibbonLinkOrder:NiagaraFloat=&ReturnNormalizedExecIndex()) {isSetParameters=true}
  ]
  particleUpdateScript [
   &ParticleState()
   /SetVariables(Particles.LinkToTargetPosition:NiagaraPosition=&Lerp_Vector(
     A=Linked() {linkedVariable {name=Engine.Emitter.SimulationPosition, type=NiagaraPosition}}
     B=Linked() {linkedVariable {name=User.Target, type=NiagaraPosition}}
     Alpha=&ReturnNormalizedExecIndex()
    )) {isSetParameters=true}
   &self:ShrinkPosCalc(
    CircleAndExtendBlend:NiagaraFloat=$System.CircleToExtendBlend
    CircleFreq:NiagaraFloat=$System.CircleFreq
    CircleRadius:NiagaraFloat=&Add_Float(A=$User.Radius, B=$Emitter.RadiusRandom)
    CircleToExtendPow=1.5
    ExtendDist:NiagaraFloat=&Add_Float(A=$User.Extend, B=$Emitter.ExtendRandom)
    InverseDir:NiagaraBool=&SetBoolByIntComparison(A=$User.InverseSineDir, B=1, Comparison Type=::NewEnumerator4)
    RotateAngle:NiagaraFloat=&Add_Float(A=$User.A, B=$Emitter.AngleRandom)
   ) {module=ShrinkPosCalc}
   /SetVariables(Particles.BlendedPosition:NiagaraPosition=&Lerp_Vector(
     A=Linked() {linkedVariable {name=Particles.LinkToTargetPosition, type=NiagaraPosition}}
     B=Linked() {linkedVariable {name=Particles.ShrinkPosition, type=NiagaraPosition}}
     Alpha=$System.BlendAB
    )) {isSetParameters=true}
   &InterpolateOverTime(
    Rate Of Change:NiagaraFloat=&Lerp_Float(
     A=128
     B=&Lerp_Float(A=4, B=32, Alpha=$System.Soft)
     Alpha=&Power(a=&ReturnNormalizedExecIndex(), b=0.15)
    )
    Target Vector Value:NiagaraPosition=$Particles.BlendedPosition
    Variable Type=::NewEnumerator2
   )
   /SetVariables(Particles.Position=$Particles.InterpolateOverTime.InterpolateOverTime_V3_.Moving Average) {isSetParameters=true}
   &CalculateAccurateVelocity()
   /SetVariables(Particles.RibbonWidth_T=2.6) {isSetParameters=true}
   &self:ReadEndPos(ReadLastPos=DataInterface() {propertyValues {EmitterBinding {bindingMode=Other, emitterName=Empty}}}) {module=ReadEndPos}
   &self:FindYAxis(Target:NiagaraPosition=$User.Target) {module=FindYAxis}
   &DynamicMaterialParameters(
    Index 0 Param 1=&MakeFloatFromVector(Vector=$Particles.XAxis, Channel=::NewEnumerator0)
    Index 0 Param 2=&MakeFloatFromVector(Vector=$Particles.XAxis, Channel=::NewEnumerator1)
    Index 0 Param 3=&MakeFloatFromVector(Vector=$Particles.XAxis, Channel=::NewEnumerator2)
    Index 1 Param 1=&MakeFloatFromVector(Vector=$Particles.YAxis, Channel=::NewEnumerator0)
    Index 1 Param 2=&MakeFloatFromVector(Vector=$Particles.YAxis, Channel=::NewEnumerator1)
    Index 1 Param 3=&MakeFloatFromVector(Vector=$Particles.YAxis, Channel=::NewEnumerator2)
    Index 2 Param 1=&MakeFloatFromVector(Vector=Linked() {linkedVariable {name=Engine.Emitter.SimulationPosition, type=NiagaraPosition}}, Channel=::NewEnumerator0)
    Index 2 Param 2=&MakeFloatFromVector(Vector=Linked() {linkedVariable {name=Engine.Emitter.SimulationPosition, type=NiagaraPosition}}, Channel=::NewEnumerator1)
    Index 2 Param 3=&MakeFloatFromVector(Vector=Linked() {linkedVariable {name=Engine.Emitter.SimulationPosition, type=NiagaraPosition}}, Channel=::NewEnumerator2)
    Index 3 Param 1=&MakeFloatFromVector(Vector=Linked() {linkedVariable {name=Particles.LastParticlePos, type=NiagaraPosition}}, Channel=::NewEnumerator0)
    Index 3 Param 2=&MakeFloatFromVector(Vector=Linked() {linkedVariable {name=Particles.LastParticlePos, type=NiagaraPosition}}, Channel=::NewEnumerator1)
    Index 3 Param 3=&MakeFloatFromVector(Vector=Linked() {linkedVariable {name=Particles.LastParticlePos, type=NiagaraPosition}}, Channel=::NewEnumerator2)
    Index 3 Param 4=&FloatFromCurve(
     FloatCurve=DataInterface() {propertyValues {Curve {keys [{interpMode=RCIM_Cubic, tangentMode=RCTM_Break, time=0, value=0}, <at 0.3 1>]}}}
     CurveIndex=$System.Age
    )
    Module.Param0WriteEnabled=true
    Module.Param0WriteEnabled1=true
    Module.Param0WriteEnabled2=true
    Module.Param0WriteEnabled3=true
    Module.Param1WriteEnabled=true
    Module.Param1WriteEnabled1=true
    Module.Param1WriteEnabled2=true
    Module.Param1WriteEnabled3=true
    Module.Param2WriteEnabled=true
    Module.Param2WriteEnabled1=true
    Module.Param2WriteEnabled2=true
    Module.Param2WriteEnabled3=true
    Module.Param3WriteEnabled=true
    Module.Param3WriteEnabled1=true
    Module.Param3WriteEnabled2=true
    Module.Param3WriteEnabled3=true
    Write Parameter Index 1=true
    Write Parameter Index 2=true
    Write Parameter Index 3=true
   )
   /SetVariables(Particles.GlowWitdh=25) {isSetParameters=true}
   &ScaleColor(
    Scale Alpha=&FloatFromCurve(
     FloatCurve=DataInterface() {propertyValues {Curve {keys [
         {interpMode=RCIM_Cubic, tangentMode=RCTM_Break, time=0.2, value=0}
         {arriveTangent=1.5665, interpMode=RCIM_Cubic, tangentMode=RCTM_Break, time=0.5, value=1}
        ]}}}
     Scale Curve=$Emitter.ColorShaking
    )
    ScaleA=true
    ScaleRGB=true
   )
  ]
  rendererClasses [&NiagaraRibbonRendererProperties, &NiagaraSpriteRendererProperties]
  renderers [
   {rendererClass=&NiagaraRibbonRendererProperties}
   {rendererClass=&NiagaraRibbonRendererProperties, rendererIndex=1}
   {rendererClass=&NiagaraSpriteRendererProperties, rendererIndex=2}
  ]
 }}
summary {
 systemName=NS_ShrinkRopeTest_NyxGlass_HookVer
 emitters [{emitterName=Empty, rendererClasses [&NiagaraRibbonRendererProperties, &NiagaraSpriteRendererProperties]}]
 userVariables [
  User.A {type=NiagaraFloat, defaultValue=0, description=""}
  User.Extend {type=NiagaraFloat, defaultValue=300, description=""}
  User.InverseSineDir {type=NiagaraInt32}
  User.Radius {type=NiagaraFloat, defaultValue=60, description=""}
  User.ScaleAge {type=NiagaraFloat, defaultValue=1, description=""}
  User.Target {type=NiagaraPosition, description="", defaultValue {}}
 ]
}
