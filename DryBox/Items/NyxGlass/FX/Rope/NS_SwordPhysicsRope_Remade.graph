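# Niagara ribbon rope for the NyxGlass sword, hanging between the effect's position and an offset point sampled on a skeletal mesh, mirrored left or right.
# A burst of particles forms the chain, simulated with gravity and custom two-point constraint and event-writing modules.
# Point count and loop factor vary by platform quality.
# Owner speed increases slack and width, owner Z scale gates visibility, and both endpoints feed dynamic material parameters.
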NiagaraSystem /DryBox/Items/NyxGlass/FX/Rope/NS_SwordPhysicsRope_Remade
dependencies {
 usedDataInterfaces [&NiagaraDataInterfacePlatformSet]
 usedDynamicInputs []
 usedModules [
  &SystemState
  &self:SystemSpawnScript.SystemScriptSource.SystemScriptGraph.NiagaraNodeAssignment_0.SetVariables_#1
  &SampleSkeletalMesh
 ]
}
emitters {Rope {
  emitterUpdateScript [
   &EmitterState(
    Inactive Response=::NewEnumerator1
    Life Cycle Mode=::NewEnumerator1
    Loop Behavior=::NewEnumerator1
    Loop Duration Mode=::NewEnumerator1
    UseLoopDelay=false
   )
   /SetVariables(Emitter.Target:NiagaraPosition=&AddVectorToPosition(
     Position=$System.SampleSkeletalMesh.SampledPosition
     Vector=&RotateVectorByQuaternion(
      Vector To Rotate=&Multiply_Vector(A=<-3.5 5.5 -10.5>, B=$System.LRScale)
      Quaternion=$System.SampleSkeletalMesh.SampledRotation
     )
    )) {isSetParameters=true}
   /SetVariables(Emitter.RopeSpawnCount:NiagaraInt32=&self:ScratchDynamicInput(
     HighP:NiagaraInt32=12
     LowP:NiagaraInt32=6
     LowPSelect:NiagaraDataInterfacePlatformSet=DataInterface() {propertyValues {Platforms {qualityLevelMask=-29}}}
    )) {isSetParameters=true}
   /SetVariables(Emitter.RopeTotalLength:NiagaraFloat=&Multiply_Float(
     A=&VectorLength(Vector=&SubtractVectorFromPosition(Position=$Emitter.Target, Vector=Linked() {linkedVariable {name=Engine.Emitter.SimulationPosition, type=NiagaraPosition}}))
     B=&Lerp_Float(
      A=1.1
      B=1.6
      Alpha=&Add_Float(
       A=&Power(
        a=&ClampFloat(Float=&Divide_Float(A=&VectorLength(Vector=$Engine.Owner.Velocity), B=2000))
        b=3
       )
       B=0
      )
     )
    )) {isSetParameters=true}
   /SetVariables(Emitter.RopeTotalLength:NiagaraFloat=&Lerp_Float(
     A=100
     B=120
     Alpha=&Add_Float(
      A=&Power(
       a=&ClampFloat(Float=&Divide_Float(A=&VectorLength(Vector=$Engine.Owner.Velocity), B=2000))
       b=3
      )
      B=0
     )
    )) {enabled=false, isSetParameters=true}
   /SetVariables(Emitter.RopeSegLength:NiagaraFloat=&Divide_Float(
     A=$Emitter.RopeTotalLength
     B=&MakeFloatFromInt(Int=/Niagara/DynamicInputs/Subtract/Subtract_Integer(A:NiagaraInt32=$Emitter.RopeSpawnCount, B:NiagaraInt32=1))
    )) {isSetParameters=true}
   &SpawnBurst_Instantaneous(Spawn Count=$Emitter.RopeSpawnCount, Use Spawn Probability=false, UseLoopCountLimit=false)
   /SetVariables(Emitter.RootVector:Vector3f=&RotateVectorByQuaternion(
     Vector To Rotate=&Multiply_Vector(
      A=&NormalizeVector(Vector To Normalize=<0 1 -1>, Safe Normalize=true, Safe Fallback Threshold=0, Safe Fallback Vector=<0 0 0.001>)
      B=&MakeCustomVectorFromBool(Bool=$User.bIsRight, True Vector=<-1 1 -1>, False Vector=<-1 -1 -1>)
     )
     Quaternion=$Engine.Owner.Rotation
    )) {isSetParameters=true}
   /SetVariables(Emitter.RootVector2:Vector3f=&RotateVectorByQuaternion(
     Vector To Rotate=&Multiply_Vector(
      A=&NormalizeVector(Vector To Normalize=<0 1 -1>, Safe Normalize=true, Safe Fallback Threshold=0, Safe Fallback Vector=<0 0 0.001>)
      B=$System.LRScale
     )
     Quaternion=$System.SampleSkeletalMesh.SampledRotation
    )) {isSetParameters=true}
   /SetVariables(Emitter.RopeDir:Vector3f=&NormalizeVector(
     Vector To Normalize=&SubtractVectorFromPosition(Position=$Emitter.Target, Vector=Linked() {linkedVariable {name=Engine.Emitter.SimulationPosition, type=NiagaraPosition}})
     Safe Normalize=true
     Safe Fallback Threshold=0
     Safe Fallback Vector=<0 0 0.001>
    )) {isSetParameters=true}
  ]
  particleSpawnScript [
   &InitializeParticle(
    Color Mode=::NewEnumerator0
    Lifetime=5
    Mesh Scale Mode=::NewEnumerator3
    Mesh Uniform Scale=10
    Ribbon Width=1.75
    Ribbon Width Mode=::NewEnumerator1
    Sprite Size Mode=::NewEnumerator3
    Uniform Sprite Size=5
    UsePositionOffset=false
    Write Material Random=true
   )
   /SetVariables(Particles.RopeMyDist:NiagaraFloat=&Multiply_Float(A=$Emitter.RopeSegLength, B=&MakeFloatFromInt(Int=&ReturnExecIndex()))) {isSetParameters=true}
   /SetVariables(Particles.RibbonLinkOrder:NiagaraFloat=&ReturnNormalizedExecIndex()) {isSetParameters=true}
   /SetVariables(Particles.Position=&AddVectorToPosition(
     Position=$Engine.Emitter.SimulationPosition
     Vector=&Multiply_VectorByFloat(
      Vector=&SubtractVectorFromPosition(Position=$Emitter.Target, Vector=Linked() {linkedVariable {name=Engine.Emitter.SimulationPosition, type=NiagaraPosition}})
      Float=$Particles.RopeMyDist
     )
    )) {isSetParameters=true}
  ]
  particleUpdateScript [
   &ParticleState(Kill Particles When Lifetime Has Elapsed=false, Let Infinitely Lived Particles Die When Emitter Deactivates=true)
   &ScaleRibbonWidth(Ribbon Width Scale=&Lerp_Float(
     A=1
     B=1.4
     Alpha=&Add_Float(
      A=&Power(
       a=&ClampFloat(Float=&Divide_Float(A=&VectorLength(Vector=$Engine.Owner.Velocity), B=2000))
       b=3
      )
      B=0
     )
    ))
   &GravityForce()
   &SolveForcesAndVelocity(Clamp Velocity=false, Limit Acceleration=false)
   &self:PhysicsRope_TwoPointConstraint(SegLength:NiagaraFloat=$Emitter.RopeSegLength, TargetPos:NiagaraPosition=$Emitter.Target) {module=PhysicsRope_TwoPointConstraint}
   &self:PhysicsRope_WriteEvent(LoopFac:NiagaraInt32=&self:ScratchDynamicInput(
     HighP:NiagaraInt32=2
     LowP:NiagaraInt32=3
     LowPSelect:NiagaraDataInterfacePlatformSet=DataInterface() {propertyValues {Platforms {qualityLevelMask=-29}}}
    )) {module=PhysicsRope_WriteEvent}
   &DynamicMaterialParameters(
    Index 2 Param 1=&MakeFloatFromVector(Vector=Linked() {linkedVariable {name=Engine.Emitter.SimulationPosition, type=NiagaraPosition}}, Channel=::NewEnumerator0)
    Index 2 Param 2=&MakeFloatFromVector(Vector=Linked() {linkedVariable {name=Engine.Emitter.SimulationPosition, type=NiagaraPosition}}, Channel=::NewEnumerator1)
    Index 2 Param 3=&MakeFloatFromVector(Vector=Linked() {linkedVariable {name=Engine.Emitter.SimulationPosition, type=NiagaraPosition}}, Channel=::NewEnumerator2)
    Index 3 Param 1=&MakeFloatFromVector(Vector=Linked() {linkedVariable {name=Emitter.Target, type=NiagaraPosition}}, Channel=::NewEnumerator0)
    Index 3 Param 2=&MakeFloatFromVector(Vector=Linked() {linkedVariable {name=Emitter.Target, type=NiagaraPosition}}, Channel=::NewEnumerator1)
    Index 3 Param 3=&MakeFloatFromVector(Vector=Linked() {linkedVariable {name=Emitter.Target, type=NiagaraPosition}}, Channel=::NewEnumerator2)
    Module.Param0WriteEnabled=true
    Module.Param0WriteEnabled2=true
    Module.Param0WriteEnabled3=true
    Module.Param1WriteEnabled=true
    Module.Param1WriteEnabled2=true
    Module.Param1WriteEnabled3=true
    Module.Param2WriteEnabled=true
    Module.Param2WriteEnabled2=true
    Module.Param2WriteEnabled3=true
    Module.Param3WriteEnabled=true
    Module.Param3WriteEnabled2=true
    Module.Param3WriteEnabled3=true
    Write Parameter Index 2=true
    Write Parameter Index 3=true
   )
   &ScaleRibbonWidth(Ribbon Width Scale=&MakeCustomFloatFromBool(
     Bool=&SetBoolByFloatComparison(A=&MakeFloatFromVector(Vector=$Engine.Owner.Scale, Channel=::NewEnumerator2), B=0.9)
     True Float=1
     False Float=0
    )) {module=ScaleRibbonWidth001}
  ]
  rendererClasses [&NiagaraSpriteRendererProperties, &NiagaraRibbonRendererProperties]
  renderers [
   {rendererClass=&NiagaraSpriteRendererProperties}
   {rendererClass=&NiagaraRibbonRendererProperties, rendererIndex=1}
  ]
 }}
summary {
 systemName=NS_SwordPhysicsRope_Remade
 emitters [{emitterName=Rope, rendererClasses [&NiagaraSpriteRendererProperties, &NiagaraRibbonRendererProperties]}]
 userVariables [
  User.bIsRight {type=NiagaraBool}
  User.PlayerSM {
   type=NiagaraDataInterfaceSkeletalMesh
   defaultValue {
    struct=NiagaraExt_VariableValue_DataInterface
    value {dataInterface=&self:NiagaraDataInterfaceSkeletalMesh_0}
   }
  }
 ]
}
